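# Layered, high-priority ambient sound cue for the Mooney event aura, mixing two branches with separate attenuation assets.
# The map-wide focus sky-object branch plays the main aura loop plus a Mooney sound repeating after random delays of tens of seconds.
# The Mooney_Aura_Att branch plays the distant loop beneath boosted, randomly picked ambient one-shots that repeat every few seconds.
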SoundCue /Figment_S07/Athena/Sounds/Events/Mooney/Mooney_Aura_Cue
root=Mixer_2
nodes [
 {
  id=Attenuation_0
  children [Mixer_0]
  properties {attenuationSettings=/Game/Athena/Sounds/Events/Attenuations/Events_Mapwide_Focus_SkyObject_Att}
 }
 {id=Attenuation_1, children [Mixer_1], properties {attenuationSettings=&/Mooney_Aura_Att}}
 {id=Delay_0, children [WavePlayer_1], properties {delayMax=90, delayMin=30}}
 {id=Delay_1, children [Random_0], properties {delayMax=5, delayMin=2.5}}
 {id=Looping_0, children [Delay_0]}
 {id=Looping_1, children [Delay_1]}
 {id=Mixer_0, children [WavePlayer_0,Looping_0], properties {inputVolume [1,0.75]}}
 {id=Mixer_1, children [Looping_1,WavePlayer_7], properties {inputVolume [2,0.5]}}
 {id=Mixer_2, children [Attenuation_0, Attenuation_1], properties {inputVolume [1,1]}}
 {id=Random_0, children [WavePlayer_2,WavePlayer_3,WavePlayer_4,WavePlayer_5], properties {weights [1,1,1,1]}}
 {id=WavePlayer_0, properties {bLooping=true, soundWaveAssetPtr=&/Mooney_Aura_Loop_01}}
 {id=WavePlayer_1, properties {soundWaveAssetPtr=&/Mooney_Mooney_01}}
 {id=WavePlayer_2, properties {soundWaveAssetPtr=&/Mooney_Aura_AmbOneShot_04}}
 {id=WavePlayer_3, properties {soundWaveAssetPtr=&/Mooney_Aura_AmbOneShot_01}}
 {id=WavePlayer_4, properties {soundWaveAssetPtr=&/Mooney_Aura_AmbOneShot_02}}
 {id=WavePlayer_5, properties {soundWaveAssetPtr=&/Mooney_Aura_AmbOneShot_03}}
 {id=WavePlayer_7, properties {bLooping=true, soundWaveAssetPtr=&/Mooney_Aura_Loop_Distant_01}}
]
properties {
 maxDistance=2097152
 priority=10
 soundClassObject=&Fort_Events
 volumeMultiplier=1
 attenuationOverrides {
  bAttenuate=false
  bEnableFocusInterpolation=true
  bEnableListenerFocus=true
  falloffDistance=1000000000
  focusAzimuth=20
  nonFocusAzimuth=40
  nonFocusVolumeAttenuation=0.6
  stereoSpread=25000
 }
}
