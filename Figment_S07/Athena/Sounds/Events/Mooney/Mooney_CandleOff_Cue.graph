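# Sound cue for the Mooney event's candle-off sound, assigned to the Fort_Events sound class.
# A Random node picks one of two equally weighted close-perspective wave variations.
# The pick feeds an Attenuation node using the shared map-wide events attenuation, which the cue also references at top level alongside stored override values for no distance attenuation and listener focus.
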SoundCue /Figment_S07/Athena/Sounds/Events/Mooney/Mooney_CandleOff_Cue
root=Attenuation_0
nodes [
 {id=Attenuation_0, children [Random_0], properties {attenuationSettings=/Figment_S07/Athena/Sounds/Events/Attenuations/Events_Mapwide_Att}}
 {id=Random_0, children [WavePlayer_2,WavePlayer_3], properties {weights [1,1]}}
 {id=WavePlayer_2, properties {soundWaveAssetPtr=&/Mooney_CandleOff_Close_02}}
 {id=WavePlayer_3, properties {soundWaveAssetPtr=&/Mooney_CandleOff_Close_01}}
]
properties {
 attenuationSettings=&Events_Mapwide_Att
 maxDistance=2097152
 priority=10
 soundClassObject=&Fort_Events
 volumeMultiplier=1
 attenuationOverrides {
  bAttenuate=false
  bEnableFocusInterpolation=true
  bEnableListenerFocus=true
  falloffDistance=1000000000
  focusAzimuth=20
  nonFocusAzimuth=40
  nonFocusVolumeAttenuation=0.6
  stereoSpread=25000
 }
}
