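# Sound cue for the Mooney event sequence audio, set up for mapwide playback in the Figment Season 7 Athena events.
# A single wave player feeds the Mooney sequence sound into an attenuation node using the shared mapwide events attenuation.
# The cue uses the mapwide events sound class with boosted volume and a very large max distance.
# Its stored attenuation overrides turn off distance attenuation and define listener focus with a wide stereo spread.
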SoundCue /Figment_S07/Athena/Sounds/Events/Mooney/Mooney_Sequence_Cue
root=Attenuation_0
nodes [
 {
  id=Attenuation_0
  children [WavePlayer_6]
  properties {attenuationSettings=/Figment_S07/Athena/Sounds/Events/Attenuations/Events_Mapwide_Att}
 }
 {id=WavePlayer_6, properties {soundWaveAssetPtr=&/Mooney_Mooney_Sequence_01}}
]
properties {
 maxDistance=2097152
 priority=10
 soundClassObject=/Figment_S07/Athena/Sounds/Events/SoundClasses/Fort_Events_Mapwide
 volumeMultiplier=1.7
 attenuationOverrides {
  bAttenuate=false
  bEnableFocusInterpolation=true
  bEnableListenerFocus=true
  falloffDistance=1000000000
  focusAzimuth=20
  nonFocusAzimuth=40
  nonFocusVolumeAttenuation=0.6
  stereoSpread=25000
 }
}
