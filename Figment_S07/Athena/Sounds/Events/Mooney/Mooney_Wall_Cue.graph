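# Sound cue for the Mooney event wall: a looping "wall in front" wave in the map-wide events sound class.
# The wave's volume is driven at runtime by a continuous modulator on the PassBy parameter.
# An attenuation node overrides spatialization with very long falloff, wide stereo spread and a distance low-pass on a steep custom curve, so the wall carries across great distances while sounding duller far away.
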SoundCue /Figment_S07/Athena/Sounds/Events/Mooney/Mooney_Wall_Cue
root=Attenuation_2
nodes [
 {
  id=Attenuation_2
  children [ModulatorContinuous_0]
  properties {
   bOverrideAttenuation=true
   attenuationOverrides {
    absorptionMethod=CustomCurve
    bAttenuateWithLPF=true
    falloffDistance=480000
    lPFFrequencyAtMax=500
    lPFRadiusMax=240000
    lPFRadiusMin=6000
    stereoSpread=50000
    attenuationShapeExtents {x=0}
    customLowpassAirAbsorptionCurve {editorCurveData {keys [
       {arriveTangent=-9.3174, interpMode=RCIM_Cubic, leaveTangent=-9.3174, tangentMode=RCTM_User, time=0, value=1}
       {arriveTangent=-0.5137, interpMode=RCIM_Cubic, leaveTangent=-0.5137, tangentMode=RCTM_User, time=0.25, value=0.1}
       <at 1 0 RCIM_Cubic>
      ]}}
   }
  }
 }
 {id=ModulatorContinuous_0, children [WavePlayer_3], properties {volumeModulationParams {parameterName=PassBy}}}
 {id=WavePlayer_3, properties {bLooping=true, soundWaveAssetPtr=&/Mooney_WallInFront_01}}
]
properties {
 maxDistance=480000
 priority=10
 soundClassObject=/Figment_S07/Athena/Sounds/Events/SoundClasses/Fort_Events_Mapwide
 volumeMultiplier=1.5
 attenuationOverrides {
  bAttenuateWithLPF=true
  bEnableFocusInterpolation=true
  bEnableListenerFocus=true
  falloffDistance=480000
  focusAzimuth=20
  lPFFrequencyAtMax=500
  lPFRadiusMax=240000
  lPFRadiusMin=120000
  nonFocusAzimuth=40
  nonFocusVolumeAttenuation=0.6
  stereoSpread=25000
 }
}
