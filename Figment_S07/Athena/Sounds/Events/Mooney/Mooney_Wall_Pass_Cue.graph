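# Sound cue for the Mooney event's wall pass, playing the Mooney_WallSweep_01 wave through a single wave player.
# It routes to the Fort_Events_Mapwide sound class with a raised priority and volume.
# Overridden attenuation spans map-scale falloff with a wide stereo spread, a low-pass filter that closes in over distance, and listener focus that lowers volume when off-focus.
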SoundCue /Figment_S07/Athena/Sounds/Events/Mooney/Mooney_Wall_Pass_Cue
root=WavePlayer_5
nodes [{id=WavePlayer_5, properties {soundWaveAssetPtr=&/Mooney_WallSweep_01}}]
properties {
 maxDistance=2097152
 priority=10
 soundClassObject=/Figment_S07/Athena/Sounds/Events/SoundClasses/Fort_Events_Mapwide
 volumeMultiplier=1.2
 attenuationOverrides {
  bAttenuateWithLPF=true
  bEnableFocusInterpolation=true
  bEnableListenerFocus=true
  falloffDistance=480000
  focusAzimuth=20
  lPFFrequencyAtMax=500
  lPFRadiusMax=240000
  lPFRadiusMin=120000
  nonFocusAzimuth=40
  nonFocusVolumeAttenuation=0.6
  stereoSpread=25000
 }
}
